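# Translucent UI material that draws the "Technical Lines" background for Item Shop, colored by three colors from the shared background-params function.
# It layers gradients, grunge, a technical-lines texture, noise and slowly rotating procedural light rays, finishing with a color-dodge blend.
# Slightly offset resamples of the lines texture, differenced with the original, create line glow and shadow; its green channel supplies a grunge mask.
# UI UV-scaling functions handle fit and fill, and a shared function handles rounded corners.
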Material /Game/Catalog/BackgroundMaterialLibrary/CoreMaterials/TechnicalLines/M_UI_Shop_Background_TechnicalLines
graph {
 properties {bAutomaticallySetUsageInEditor=false, blendMode=BLEND_Translucent, materialDomain=MD_UI, usedShadingModels=Default Lit}
 nodes [
  MaterialFunctionCall!(Input Alpha=Constant() {r=1}, bOverrideCornerRadius, Corner Radius Override) {materialFunction=/FortUILibrary/Systems/ItemShopTile/Materials/MaterialFunctions/MF_UI_ItemShop_RoundedCorners}
  n1 = MaterialFunctionCall() {materialFunction=/FortUILibrary/Systems/ItemShopTile/Materials/MaterialFunctions/MF_UI_ItemShop_BackgroundParams}
  NamedRerouteDeclaration(None=@n1.Color 1) {name=Primary Color, nodeColor=<rgba 0 0 0 1>}
  n2 = NamedRerouteUsage()
  n3 = ScalarParameter() {defaultValue=1.5, group=Radial Gradient, parameterName=Radial Gradient Scale}
  n4 = ScalarParameter() {group=Radial Gradient, parameterName=Radial Gradient Offset X}
  n5 = ScalarParameter() {defaultValue=-0.1, group=Radial Gradient, parameterName=Radial Gradient Offset Y}
  n6 = MaterialFunctionCall(UV=Add(
    A=MaterialFunctionCall(UVs=@n2, Texture Scale=@n3) {materialFunction=&ScaleUVsByCenter}.UVs
    B=AppendVector(A=@n4, B=@n5)
   ) {constB=0.5}) {materialFunction=&MF_UI_UVScaling}
  NamedRerouteDeclaration(None=Saturate(None=MaterialFunctionCall(UV=@n6.Scaled To Fill) {materialFunction=&MF_UI_RadialGradient}.Smooth)) {name=Radial Gradient Mask, nodeColor=<rgba 0 0 0 1>}
  n7 = NamedRerouteUsage()
  n8 = Saturate(None=MaterialFunctionCall(Input=ComponentMask(None=@n7) {g=true}, Input Low=Constant(), Input High=Constant() {r=0.4}) {materialFunction=&RemapValue_0-1}.Result)
  NamedRerouteDeclaration(None=LinearInterpolate(A=NamedRerouteUsage(), B=NamedRerouteUsage(), Alpha=SmoothStep(Value=@n8))) {name=Background Gradient, nodeColor=<rgba 0 0 0 1>}
  n9 = NamedRerouteUsage()
  n10 = TextureObject() {
   samplerType=SAMPLERTYPE_Alpha
   texture=/Game/Catalog/BackgroundMaterialLibrary/CoreMaterials/CrumpleFold/Textures/T_UI_DarkVortex_Background_Grunge
  }
  n11 = TextureSample(
   UVs=MaterialFunctionCall(UVs=@n9, Texture Scale=Constant() {r=0.3}) {materialFunction=&ScaleUVsByCenter}.UVs
   Tex=@n10
  ) {
   samplerType=SAMPLERTYPE_Alpha
   texture=&T_UI_DarkVortex_Background_Grunge
  }
  n12 = TextureSample(UVs=@n9, Tex=TextureObject() {samplerType=SAMPLERTYPE_Grayscale, texture=&LowResBlurredNoise}) {samplerType=SAMPLERTYPE_Grayscale}
  n13 = Multiply(A=@n11.R, B=LinearInterpolate(Alpha=Multiply(A=@n12.R, B=@n12.R)) {constA=0.1})
  NamedRerouteDeclaration(None=Saturate(None=Multiply(A=@n13) {constB=2})) {name=Grunge Texture, nodeColor=<rgba 0 0 0 1>}
  n14 = MaterialFunctionCall(UV=NamedRerouteDeclaration(None=MaterialFunctionCall() {materialFunction=&GetUserInterfaceUV}.Normalized UV) {name=Normalized UV, nodeColor=<rgba 0 0 0 1>}) {materialFunction=&MF_UI_UVScaling}
  NamedRerouteDeclaration(None=@n14.Scaled to Fit) {name=Scale to Fit, nodeColor=<rgba 0 0 0 1>}
  n15 = NamedRerouteUsage()
  n16 = NamedRerouteUsage()
  n17 = ScalarParameter() {defaultValue=-0.01, group=Lines Texture, parameterName=Texture Position X}
  n18 = ScalarParameter() {defaultValue=-0.11, group=Lines Texture, parameterName=Texture Position Y}
  n19 = AppendVector(A=ComponentMask(None=Add(A=@n16, B=@n17) {constB=0.2}) {r=true}, B=ComponentMask(None=Add(A=@n16, B=@n18) {constB=0.3}) {g=true})
  n20 = TextureSample(UVs=MaterialFunctionCall(UVs=@n19, Texture Scale=ScalarParameter() {defaultValue=0.35, group=Lines Texture, parameterName=Texture Size}) {materialFunction=&ScaleUVsByCenter}.UVs) {samplerType=SAMPLERTYPE_LinearColor, texture=/Game/Catalog/BackgroundMaterialLibrary/Textures/T_UI_TechnicalLines}
  n21 = Saturate(None=@n20.R)
  NamedRerouteDeclaration(None=LinearInterpolate(B=LinearInterpolate(B=@n15, Alpha=@n21), Alpha=NamedRerouteUsage())) {name=Lines Color, nodeColor=<rgba 0 0 0 1>}
  NamedRerouteDeclaration(None=@n1.Color 2) {name=Secondary Color, nodeColor=<rgba 0 0 0 1>}
  NamedRerouteDeclaration(None=Saturate(None=@n20.G)) {name=Grunge Texture Mask, nodeColor=<rgba 0 0 0 1>}
  n22 = Multiply(
   A=MaterialFunctionCall!(UV=NamedRerouteUsage(), Seed) {materialFunction=&MF_UI_Random}.Result
   B=ScalarParameter() {defaultValue=0.3, group=Noise, parameterName=Noise Opacity}
  ) {constB=0.65}
  n23 = ComponentMask(None=NamedRerouteUsage()) {g=true}
  n24 = LinearInterpolate(Alpha=Power(Base=OneMinus(None=@n23))) {constA=0.1}
  NamedRerouteDeclaration(None=Saturate(None=Multiply(A=@n22, B=@n24))) {name=Noise Overlay, nodeColor=<rgba 0 0 0 1>}
  n25 = MaterialFunctionCall!(ScreenUV, TextureSize, ScalePivotPoint) {materialFunction=/FortUILibrary/MaterialFunctions/Basics/MF_UI_ScreenSpaceUVScaling}
  NamedRerouteDeclaration(None=@n25.ScaleY) {name="Scale Y UV  Screen", nodeColor=<rgba 0 0 0 1>}
  NamedRerouteDeclaration(None=@n25.Scaled to Fill) {name="Scale to Fill  Screen", nodeColor=<rgba 0 0 0 1>}
  NamedRerouteDeclaration(None=@n25.Scaled to Fit) {name=Scale to Fit Screen, nodeColor=<rgba 0 0 0 1>}
  NamedRerouteDeclaration(None=@n14.ScaleY) {name=Scale Y UV, nodeColor=<rgba 0 0 0 1>}
  n26 = ComponentMask(None=Add(A=@n19) {constB=0.0012}) {r=true}
  n27 = ComponentMask(None=Add(A=@n19) {constB=-0.0012}) {g=true}
  n28 = TextureSample(UVs=MaterialFunctionCall(UVs=AppendVector(A=@n26, B=@n27), Texture Scale=ScalarParameter() {defaultValue=0.35, group=Lines Texture, parameterName=Texture Size}) {materialFunction=&ScaleUVsByCenter}.UVs) {samplerType=SAMPLERTYPE_LinearColor, texture=&T_UI_TechnicalLines}
  NamedRerouteDeclaration(None=Saturate(None=Subtract(A=@n28.R, B=@n20.R))) {name=Lines Glow, nodeColor=<rgba 0 0 0 1>}
  n29 = Add(A=@n19) {constB=-0.0008}
  n30 = Add(A=@n19) {constB=0.0008}
  n31 = MaterialFunctionCall(
   UVs=AppendVector(A=ComponentMask(None=@n29) {r=true}, B=ComponentMask(None=@n30) {g=true})
   Texture Scale=ScalarParameter() {defaultValue=0.35, group=Lines Texture, parameterName=Texture Size}
  ) {materialFunction=&ScaleUVsByCenter}
  n32 = OneMinus(None=Subtract(
    A=TextureSample(UVs=@n31.UVs) {samplerType=SAMPLERTYPE_LinearColor, texture=&T_UI_TechnicalLines}.R
    B=@n20.R
   ))
  NamedRerouteDeclaration(None=LinearInterpolate(A=Constant() {r=0.2}, Alpha=Saturate(None=@n32)) {constA=0.3}) {name=Lines Shadow, nodeColor=<rgba 0 0 0 1>}
  n33 = NamedRerouteUsage()
  n34 = Constant2Vector() {g=0.4}
  n35 = ScalarParameter() {defaultValue=0.02, group=Light Beam, parameterName=Light Rays Speed}
  n36 = Time()
  n37 = MaterialFunctionCall!(Vector or UVs=Rotator(Coordinate=Add(A=@n33, B=@n34), Time=Multiply(A=@n35, B=@n36)) {speed=6.2832}) {materialFunction=/FortUILibrary/MaterialFunctions/Basics/MF_UI_CheaperVectorToRadial}
  n38 = NamedRerouteUsage()
  n39 = NamedRerouteDeclaration(None=TextureObject() {samplerType=SAMPLERTYPE_Grayscale, texture=&LowResBlurredNoise}) {name=Rays Texture, nodeColor=<rgba 0 0 0 1>}
  n40 = TextureSample(UVs=Multiply(A=@n37.Vector Converted to Angle, B=Constant2Vector() {g=1, r=-1}) {constB=2}, Tex=@n38)
  n41 = Min(
   A=TextureSample(UVs=TextureSample(UVs=@n37.Vector Converted to Angle, Tex=@n38).R, Tex=@n39).R
   B=TextureSample(UVs=Multiply(A=@n40.R) {constB=1.5}, Tex=@n39).R
  )
  n42 = LinearInterpolate(Alpha=Multiply(A=SmoothStep(Value=@n41) {constMax=0.7, constMin=0.3}) {constB=3})
  NamedRerouteDeclaration(None=Saturate(None=OneMinus(None=@n42))) {name=Light Rays, nodeColor=<rgba 0 0 0 1>}
  n43 = MaterialFunctionCall(Input=ComponentMask(None=NamedRerouteUsage()) {g=true}, Input Low=Constant(), Input High=Constant() {r=1}) {materialFunction=&RemapValue_0-1}
  NamedRerouteDeclaration(None=SmoothStep(Value=Saturate(None=@n43.Result)) {constMin=0.4}) {name=Bottom Gradient, nodeColor=<rgba 0 0 0 1>}
  NamedRerouteDeclaration(None=@n14.ScaleX) {name=Scale X UV, nodeColor=<rgba 0 0 0 1>}
  NamedRerouteDeclaration(None=@n14.Scaled To Fill) {name=Scale to Fill, nodeColor=<rgba 0 0 0 1>}
  NamedRerouteDeclaration(None=@n1.Color 3) {name=Accent Color, nodeColor=<rgba 0 0 0 1>}
  n44 = NamedRerouteUsage()
  n45 = NamedRerouteUsage()
  n46 = NamedRerouteUsage()
  n47 = NamedRerouteUsage()
  n48 = NamedRerouteUsage()
  n49 = Add(
   A=LinearInterpolate(A=@n44, B=Multiply(A=@n44) {constB=0.25}, Alpha=LinearInterpolate(B=@n45, Alpha=@n46))
   B=Multiply(A=LinearInterpolate(B=@n47, Alpha=@n48)) {constB=0.3}
  ) {constB=0}
  n50 = Multiply(A=NamedRerouteUsage()) {constB=0.8}
  n51 = LinearInterpolate(A=NamedRerouteUsage(), B=LinearInterpolate(B=Add(A=@n49, B=@n50), Alpha=NamedRerouteUsage()) {constB=0}, Alpha=NamedRerouteUsage()) {constA=1}
  n52 = NamedRerouteUsage()
  n53 = NamedRerouteUsage()
  Saturate(None=MaterialFunctionCall!(Blend=NamedRerouteUsage(), Base=LinearInterpolate(A=@n51, B=@n52, Alpha=@n53)) {materialFunction=/Engine/Functions/Engine_MaterialFunctions03/Blends/Blend_ColorDodge}.Result)
 ]
}
